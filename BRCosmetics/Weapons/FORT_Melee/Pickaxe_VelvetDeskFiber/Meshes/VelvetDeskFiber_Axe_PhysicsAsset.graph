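# Physics asset for the VelvetDeskFiber pickaxe, a two-handed melee weapon cosmetic.
# A single capsule body on the Melee_TwoHanded_Handle bone approximates the weapon's collision shape along the handle.
# It defines no constraints, so the weapon simulates as one rigid piece.
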PhysicsAsset /BRCosmetics/Weapons/FORT_Melee/Pickaxe_VelvetDeskFiber/Meshes/VelvetDeskFiber_Axe_PhysicsAsset
bodies {Melee_TwoHanded_Handle [{shapeType=Capsule, length=83.9642, radius=14.2105, center=<0.9519 -1.3946 45.3824>, rotation=<rot 0.2669 -90 -0.0199>}]}
constraints []
